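# Material function that applies a recolorable tie-dye look, within an input mask, to material attributes for the F_MED_Tie_Dye_Fashion character body.
# A pattern mask texture's RGB channels drive chained lerps that build the dye color, with a static switch swapping colors 2 and 3.
# The mask blends Base Color, feeds that result into CustomData0 as fuzz color, and blends CustomData1 toward a cloth amount parameter.
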MaterialFunction /BRCosmetics/Characters/Player/Female/Medium/Bodies/F_MED_Tie_Dye_Fashion/Materials/MF_TieDye
graph {nodes [
  n1 = GetMaterialAttributes(None=FunctionInput() {inputName=Material, inputType=FunctionInput_MaterialAttributes, sortPriority=10}) {attributeGetTypes [#1, #2, #3]}
  n2 = FunctionInput() {inputName=Mask, sortPriority=100}
  n3 = LinearInterpolate(A=NamedRerouteDeclaration(None=@n1.BaseColor) {name=BaseColor, nodeColor=<rgba 0.7813 0 1 1>}, B=NamedRerouteUsage(), Alpha=@n2) {constAlpha=1}
  n4 = ScalarParameter() {defaultValue=0.3, parameterName=TieDyeClothAmount}
  FunctionOutput(None=SetMaterialAttributes!(
    MaterialAttributes=@n1.MaterialAttributes
    BaseColor=@n3
    CustomData0=LinearInterpolate(A=@n1.CustomData0, B=@n3, Alpha=@n2) {desc=Fuzz Color}
    CustomData1=LinearInterpolate(A=@n1.CustomData1, B=@n4, Alpha=@n2) {desc=Cloth}
   ) {attributeSetTypes [#1, #2, #3]})
  n5 = VectorParameter() {parameterName=TieDye_Color_3, defaultValue=<rgba 1 0 0.0338 0>}
  n6 = VectorParameter() {parameterName=TieDye_Color_2, defaultValue=<rgba 0 0.133 0.849 0>}
  n7 = TextureSampleParameter2D() {
   parameterName=TieDye_Pattern_Mask
   samplerSource=SSM_Wrap_WorldGroupSettings
   samplerType=SAMPLERTYPE_Masks
   texture=/Game/Characters/Player/Female/Medium/Bodies/F_MED_Tie_Dye_Fashion/Textures/T_F_MED_Tie_Dye_Fashion_Body_OmbreMask
  }
  n8 = LinearInterpolate(
   A=VectorParameter() {parameterName=TieDye_Color_1, defaultValue=<rgba 0.0718 0.6979 0 0>}.RGB
   B=StaticSwitchParameter(True=@n5.RGB, False=@n6.RGB) {parameterName=SwapColors2And3}
   Alpha=@n7.R
  )
  n9 = StaticSwitchParameter(True=@n6.RGB, False=@n5.RGB) {parameterName=SwapColors2And3}
  NamedRerouteDeclaration(None=LinearInterpolate(A=LinearInterpolate(A=@n8, B=@n9, Alpha=@n7.G), B=NamedRerouteUsage(), Alpha=@n7.B)) {name=TieDyeColor, nodeColor=<rgba 0.0078 0 1 1>}
  VectorParameter() {parameterName=TieDye_Color_Base, defaultValue=<rgba 0.4853 0.6917 1 0>}
  VectorParameter() {parameterName=Black, defaultValue=<rgba 0 0 0 1>}
 ]}
