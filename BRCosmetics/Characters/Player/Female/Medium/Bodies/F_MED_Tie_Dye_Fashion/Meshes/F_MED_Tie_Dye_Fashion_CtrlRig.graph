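# Control Rig for the F_MED_Tie_Dye_Fashion female body that drives corrective offsets on the left and right glute bones as the thighs rotate.
# Each thigh's rotation relative to the pelvis becomes ZYX Euler, and the absolute Z angle is remapped to a clamped weight.
# That weight blends a translation and scale change onto the matching dyn_glute bone.
# The right side uses a reversed, different input range from the left.
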ControlRigBlueprint /BRCosmetics/Characters/Player/Female/Medium/Bodies/F_MED_Tie_Dye_Fashion/Meshes/F_MED_Tie_Dye_Fashion_CtrlRig
graphs [
 RigVMFunctionLibrary {links [], nodes []}
 RigVMModel {nodes [
   RigUnit_BeginExecution = &RigUnit_BeginExecution()
   n1 = &RigVMFunction_MathQuaternionToEuler(
    RotationOrder=ZYX
    Value=&RigUnit_GetRelativeTransformForItem(Child=(Type=Bone,Name="thigh_l"), Parent=(Type=Bone,Name="pelvis")).RelativeTransform.Rotation
   )
   &RigUnit_ModifyTransforms(
    ItemToModify=((Item=(Type=Bone,Name="dyn_glute_l"),Transform=(Rotation=(X=0,Y=0,Z=0,W=1),Translation=(X=5,Y=0,Z=-5),Scale3D=(X=2,Y=1,Z=2))))
    WeightMinimum=0
    Weight=&RigVMFunction_MathFloatRemap(
     SourceMaximum=105
     SourceMinimum=45
     TargetMinimum=0
     bClamp=true
     Value=/Script/RigVM.RigVMFunction_MathFloatAbs(Value=@n1.Result.Z) {is=struct RigVMUnitNode, template="Absolute::Execute(in Value,out Result)"}.Result
    ).Result
    ExecutePin=@RigUnit_BeginExecution.ExecutePin
   )
   n2 = &RigVMFunction_MathQuaternionToEuler(
    RotationOrder=ZYX
    Value=&RigUnit_GetRelativeTransformForItem(Child=(Type=Bone,Name="thigh_r"), Parent=(Type=Bone,Name="pelvis")).RelativeTransform.Rotation
   )
   > &RigUnit_ModifyTransforms(
    ItemToModify=((Item=(Type=Bone,Name="dyn_glute_r"),Transform=(Rotation=(X=0,Y=0,Z=0,W=1),Translation=(X=5,Y=0,Z=-5),Scale3D=(X=2,Y=1,Z=2))))
    WeightMinimum=0
    Weight=&RigVMFunction_MathFloatRemap(
     SourceMaximum=72.5
     SourceMinimum=150
     TargetMinimum=0
     bClamp=true
     Value=&RigVMFunction_MathFloatAbs(Value=@n2.Result.Z) {template="Absolute::Execute(in Value,out Result)"}.Result
    ).Result
   )
  ]}
]
hierarchy [
 root {loc=<0 0 0>, rot=<0 0 0 1>}
 pelvis {parent=root, loc=<0 -0.5301 98.0571>, rot=<0.0127 0.707 -0.0127 -0.707>}
 spine_01 {parent=pelvis, loc=<6.1208 -0.9328 0>, rot=<0 0 0.034 -0.9994>}
 spine_02 {parent=spine_01, loc=<6.5996 -0.2095 0>, rot=<0 0 0.0238 -0.9997>}
 spine_03 {parent=spine_02, loc=<6.4914 -0.1237 0>, rot=<0 0 0.0149 -0.9999>}
 spine_04 {parent=spine_03, loc=<6.9598 0.0794 0>, rot=<0 0 -0.0132 -0.9999>}
 spine_05 {parent=spine_04, loc=<6.9987 0 0>, rot=<0 0 -0.0015 -1>}
 clavicle_l {parent=spine_05, loc=<5.4831 -0.1053 -3.2946>, rot=<0.0465 -0.7289 -0.009 -0.6829>}
 upperarm_l {parent=clavicle_l, loc=<9.8658 -1.2136 -0.4105>, rot=<0.0164 -0.3245 -0.0269 -0.9454>}
 lowerarm_l {parent=upperarm_l, loc=<26.465 -0.8345 0.259>, rot=<0.0036 0.0161 -0.2298 -0.9731>}
 hand_l {parent=lowerarm_l, loc=<22.7355 -0.3203 -0.6106>, rot=<-0.6784 -0.0223 0.0735 -0.7307>}
 index_metacarpal_l {parent=hand_l, loc=<1.6535 0.0183 -2.96>, rot=<0.0158 -0.0124 -0.0173 -0.9996>}
 index_01_l {parent=index_metacarpal_l, loc=<8.3008 0.0357 -0.0524>, rot=<-0.0044 0.0485 0.0924 -0.9945>}
 index_02_l {parent=index_01_l, loc=<3.3587 -0.3004 0.0452>, rot=<-0.002 0.0027 0.3225 -0.9466>}
 index_03_l {parent=index_02_l, loc=<2.5578 -0.123 -0.0252>, rot=<-0.002 0.0067 0.2558 -0.9667>}
 middle_metacarpal_l {parent=hand_l, loc=<2.1856 -0.1725 -0.5984>, rot=<0.006 0.0168 -0.0329 -0.9993>}
 middle_01_l {parent=middle_metacarpal_l, loc=<7.3275 -0.2447 0.0129>, rot=<-0.0208 0.0042 0.0858 -0.9961>}
 middle_02_l {parent=middle_01_l, loc=<4.0123 0.1401 0.1175>, rot=<-0.0218 0.0258 0.3512 -0.9357>}
 middle_03_l {parent=middle_02_l, loc=<2.6077 0.0534 0.1373>, rot=<0.0082 0.0188 0.1708 -0.9851>}
 pinky_metacarpal_l {parent=hand_l, loc=<1.7616 -0.9753 3.3814>, rot=<-0.0397 0.0814 -0.069 -0.9935>}
 pinky_01_l {parent=pinky_metacarpal_l, loc=<7.1709 -0.4831 -0.1094>, rot=<-0.0654 -0.0035 0.2242 -0.9723>}
 pinky_02_l {parent=pinky_01_l, loc=<2.8274 0.0893 -0.1043>, rot=<0.0037 0.0011 0.2854 -0.9584>}
 pinky_03_l {parent=pinky_02_l, loc=<2.2046 0.0307 -0.0078>, rot=<-0.0002 0.0098 0.1445 -0.9895>}
 ring_metacarpal_l {parent=hand_l, loc=<2.2074 -0.3625 1.3638>, rot=<-0.0284 0.0722 -0.0458 -0.9959>}
 ring_01_l {parent=ring_metacarpal_l, loc=<7.0295 -0.6197 -0.1446>, rot=<-0.0489 -0.0252 0.1892 -0.9804>}
 ring_02_l {parent=ring_01_l, loc=<3.6174 0.4647 -0.1027>, rot=<0 0 0.2806 -0.9598>}
 ring_03_l {parent=ring_02_l, loc=<2.4152 0.114 0.0199>, rot=<0 0 0.1189 -0.9929>}
 thumb_01_l {parent=hand_l, loc=<1.3118 -0.3191 -2.459>, rot=<0.6546 -0.4042 0.0664 -0.6354>}
 thumb_02_l {parent=thumb_01_l, loc=<5.9336 -0.0448 0.0401>, rot=<-0.0037 0.0213 0.298 -0.9543>}
 thumb_03_l {parent=thumb_02_l, loc=<2.2223 0.3682 -0.0907>, rot=<0.0118 -0.013 0.0066 -0.9998>}
 lowerarm_twist_01_l {parent=lowerarm_l, loc=<18.316 0 0>, rot=<0 0 0 1>}
 dyn_lowerhoodie_l_bk {parent=lowerarm_twist_01_l, loc=<-18.316 -4 0>, rot=<-0.7071 0 0 -0.7071>}
 dyn_lowerhoodie_l_bt {parent=lowerarm_twist_01_l, loc=<-18.316 0 -4>, rot=<1 0 0 0>}
 dyn_lowerhoodie_l_fr {parent=lowerarm_twist_01_l, loc=<-18.316 4 0>, rot=<0.7071 0 0 -0.7071>}
 dyn_lowerhoodie_l_tp {parent=lowerarm_twist_01_l, loc=<-18.316 0 4>, rot=<0 0 0 1>}
 lowerarm_twist_02_l {parent=lowerarm_l, loc=<10.6707 0 0>, rot=<0 0 0 1>}
 upperarm_twist_01_l {parent=upperarm_l, loc=<7.2135 -0.1134 0>, rot=<0 0 0 1>}
 dyn_upperhoodie_l_bt {parent=upperarm_twist_01_l, loc=<-7.2135 0.1134 -4>, rot=<1 0 0 0>}
 dyn_upperhoodie_l_tp {parent=upperarm_twist_01_l, loc=<-7.2135 0.1134 4>, rot=<0 0 0 1>}
 dyn_upperhoodie_l_bk {parent=upperarm_twist_01_l, loc=<-7.2135 -3.8866 0>, rot=<-0.6533 0.2706 0.2706 -0.6533>}
 dyn_upperhoodie_l_ft {parent=upperarm_twist_01_l, loc=<-7.2135 4.1134 0>, rot=<0.7071 0 0 -0.7071>}
 upperarm_twist_02_l {parent=upperarm_l, loc=<19.7998 -0.1134 0>, rot=<0 0 0 1>}
 clavicle_r {parent=spine_05, loc=<5.4831 -0.1053 3.2946>, rot=<0.7289 0.0465 -0.6829 0.009>}
 upperarm_r {parent=clavicle_r, loc=<-9.8658 1.2136 0.4104>, rot=<0.0164 -0.3245 -0.0269 -0.9454>}
 lowerarm_r {parent=upperarm_r, loc=<-26.4651 0.8345 -0.259>, rot=<0.0036 0.0161 -0.2298 -0.9731>}
 hand_r {parent=lowerarm_r, loc=<-22.7356 0.3203 0.6106>, rot=<-0.6784 -0.0223 0.0735 -0.7307>}
 index_metacarpal_r {parent=hand_r, loc=<-1.6534 -0.0181 2.9601>, rot=<0.0158 -0.0124 -0.0173 -0.9996>}
 index_01_r {parent=index_metacarpal_r, loc=<-8.3008 -0.0357 0.0524>, rot=<-0.0044 0.0485 0.0924 -0.9945>}
 index_02_r {parent=index_01_r, loc=<-3.3589 0.3003 -0.0453>, rot=<-0.002 0.0027 0.3225 -0.9466>}
 index_03_r {parent=index_02_r, loc=<-2.5578 0.123 0.0252>, rot=<-0.002 0.0067 0.2558 -0.9667>}
 middle_metacarpal_r {parent=hand_r, loc=<-2.1856 0.1726 0.5984>, rot=<0.006 0.0168 -0.0329 -0.9993>}
 middle_01_r {parent=middle_metacarpal_r, loc=<-7.3273 0.2447 -0.0129>, rot=<-0.0208 0.0042 0.0858 -0.9961>}
 middle_02_r {parent=middle_01_r, loc=<-4.0124 -0.14 -0.1175>, rot=<-0.0218 0.0258 0.3512 -0.9357>}
 middle_03_r {parent=middle_02_r, loc=<-2.6077 -0.0533 -0.1373>, rot=<0.0082 0.0188 0.1708 -0.9851>}
 pinky_metacarpal_r {parent=hand_r, loc=<-1.7617 0.9753 -3.3814>, rot=<-0.0397 0.0814 -0.069 -0.9935>}
 pinky_01_r {parent=pinky_metacarpal_r, loc=<-7.1707 0.4832 0.1094>, rot=<-0.0654 -0.0035 0.2242 -0.9723>}
 pinky_02_r {parent=pinky_01_r, loc=<-2.8274 -0.0892 0.1043>, rot=<0.0037 0.0011 0.2854 -0.9584>}
 pinky_03_r {parent=pinky_02_r, loc=<-2.2045 -0.031 0.0078>, rot=<-0.0002 0.0098 0.1445 -0.9895>}
 ring_metacarpal_r {parent=hand_r, loc=<-2.2073 0.3627 -1.3638>, rot=<-0.0284 0.0722 -0.0458 -0.9959>}
 ring_01_r {parent=ring_metacarpal_r, loc=<-7.0297 0.6197 0.1446>, rot=<-0.0489 -0.0252 0.1892 -0.9804>}
 ring_02_r {parent=ring_01_r, loc=<-3.6171 -0.4648 0.1027>, rot=<0 0 0.2806 -0.9598>}
 ring_03_r {parent=ring_02_r, loc=<-2.4153 -0.1141 -0.0199>, rot=<0 0 0.1189 -0.9929>}
 thumb_01_r {parent=hand_r, loc=<-1.3119 0.3192 2.459>, rot=<0.6546 -0.4042 0.0664 -0.6354>}
 thumb_02_r {parent=thumb_01_r, loc=<-5.9336 0.0447 -0.0401>, rot=<-0.0037 0.0213 0.298 -0.9543>}
 thumb_03_r {parent=thumb_02_r, loc=<-2.2223 -0.3682 0.0907>, rot=<0.0118 -0.013 0.0066 -0.9998>}
 lowerarm_twist_01_r {parent=lowerarm_r, loc=<-18.3161 0 0>, rot=<0 0 0 1>}
 dyn_lowerhoodie_r_bt {parent=lowerarm_twist_01_r, loc=<18.316 0.0001 4.0003>, rot=<0 0 1 0>}
 dyn_lowerhoodie_r_bk {parent=lowerarm_twist_01_r, loc=<18.3164 3.9999 -0.0002>, rot=<0 0.7071 0.7071 0>}
 dyn_lowerhoodie_r_tp {parent=lowerarm_twist_01_r, loc=<18.3164 -0.0001 -4.0002>, rot=<0 -1 0 0>}
 dyn_lowerhoodie_r_fr {parent=lowerarm_twist_01_r, loc=<18.316 -3.9999 0.0003>, rot=<0 -0.7071 0.7071 0>}
 lowerarm_twist_02_r {parent=lowerarm_r, loc=<-10.6707 0 0>, rot=<0 0 0 1>}
 upperarm_twist_01_r {parent=upperarm_r, loc=<-7.2135 0.1134 0>, rot=<0 0 0 1>}
 dyn_upperhoodie_r_bt {parent=upperarm_twist_01_r, loc=<7.2136 -0.1134 3.9999>, rot=<0 0 1 0>}
 dyn_upperhoodie_r_tp {parent=upperarm_twist_01_r, loc=<7.2131 -0.1134 -3.9996>, rot=<0 -1 0 0>}
 dyn_upperhoodie_r_bk {parent=upperarm_twist_01_r, loc=<7.2131 3.8866 0.0004>, rot=<0 0.7071 0.7071 0>}
 dyn_upperhoodie_r_ft {parent=upperarm_twist_01_r, loc=<7.2136 -4.1134 -0.0001>, rot=<0 -0.7071 0.7071 0>}
 upperarm_twist_02_r {parent=upperarm_r, loc=<-19.7998 0.1134 0>, rot=<0 0 0 1>}
 neck_01 {parent=spine_05, loc=<11.3138 0.1529 0>, rot=<0 0 -0.2003 -0.9797>}
 neck_02 {parent=neck_01, loc=<4.5627 0 0>, rot=<0 0 0.0504 -0.9987>}
 head {parent=neck_02, loc=<5.3914 0 0>, rot=<0 0 0.115 -0.9934>}
 dyn_string_r_01 {parent=spine_05, loc=<3.9774 8.5024 2.7396>, rot=<0.0448 0.0178 0.9548 0.2935>}
 dyn_string_r_02 {parent=dyn_string_r_01, loc=<3.113 0.0002 0>, rot=<0.0004 -0.0059 -0.0623 -0.998>}
 dyn_string_r_03 {parent=dyn_string_r_02, loc=<3.2722 0.0001 0>, rot=<0 0 0 1>}
 dyn_string_l_01 {parent=spine_05, loc=<3.9771 8.5023 -2.7396>, rot=<-0.0448 -0.0178 0.9548 0.2935>}
 dyn_string_l_02 {parent=dyn_string_l_01, loc=<3.1127 0 0>, rot=<-0.0004 0.0059 -0.0623 -0.998>}
 dyn_string_l_03 {parent=dyn_string_l_02, loc=<3.272 0 0>, rot=<0 0 0 1>}
 dyn_hood_l_01 {parent=spine_05, loc=<10.4109 -2.1978 -5.9029>, rot=<0.4323 -0.5596 0.5596 -0.4323>}
 dyn_hood_l_02 {parent=dyn_hood_l_01, loc=<5.9008 0 0>, rot=<0 -0.5791 0 -0.8152>}
 dyn_hood_r_01 {parent=spine_05, loc=<10.411 -2.1978 5.9029>, rot=<0.4323 -0.5596 0.5596 -0.4323>}
 dyn_hood_r_02 {parent=dyn_hood_r_01, loc=<5.9008 0 0>, rot=<0 -0.5791 0 -0.8152>}
 dyn_hoodie_l {parent=spine_02, loc=<9.971 -0.1875 -9.136>, rot=<0.0282 -0.7065 -0.0282 -0.7065>}
 dyn_hoodie_fr {parent=spine_02, loc=<9.1424 10.1763 0>, rot=<0.4797 0.5195 0.5195 0.4797>}
 dyn_hoodie_r {parent=spine_02, loc=<9.971 -0.1875 9.136>, rot=<0.7065 0.0282 0.7065 -0.0282>}
 dyn_hoodie_bk {parent=spine_02, loc=<10.4575 -6.2743 0>, rot=<0.5195 -0.4797 0.4797 -0.5195>}
 thigh_l {parent=pelvis, loc=<-1.0855 -0.2109 -8.7053>, rot=<0.0009 0.0355 0.0155 -0.9993>}
 calf_l {parent=thigh_l, loc=<-42.5284 -0.0019 -0.7297>, rot=<0.0006 -0.0161 -0.0391 -0.9991>}
 calf_twist_01_l {parent=calf_l, loc=<-35.8483 -1.9197 -0.5794>, rot=<0.0237 0.0095 -0.0035 -0.9997>}
 calf_twist_02_l {parent=calf_l, loc=<-19.1546 -1.7943 -0.2638>, rot=<0.0237 0.0095 -0.0035 -0.9997>}
 foot_l {parent=calf_l, loc=<-46.7401 -2.5014 -0.592>, rot=<0.0427 -0.0238 0.0353 -0.9982>}
 ball_l {parent=foot_l, loc=<-3.5658 10.8066 -0.1544>, rot=<0.0001 -0.0001 -0.7063 -0.7079>}
 dyn_shoelace_MSTR_l {parent=foot_l, loc=<2.0855 6.2559 0.5611>, rot=<-0.0417 0.018 -0.3368 -0.9405>}
 dyn_shoelace_loopL_l {parent=dyn_shoelace_MSTR_l, loc=<0.2675 -0.0228 -0.5422>, rot=<-0.4263 0.46 0.6146 0.4784>}
 dyn_shoelace_stringL_l {parent=dyn_shoelace_loopL_l, loc=<0.7545 -0.1893 0.875>, rot=<-0.065 0.5365 0.0771 -0.8378>}
 dyn_shoelace_loopR_l {parent=dyn_shoelace_MSTR_l, loc=<0.3414 0.1535 0.4883>, rot=<-0.3832 0.3114 -0.6571 -0.5696>}
 dyn_shoelace_stringR_l {parent=dyn_shoelace_loopR_l, loc=<0.6987 -0.1645 -0.8907>, rot=<0.0834 -0.4974 0.0929 -0.8585>}
 thigh_twist_01_l {parent=thigh_l, loc=<-27.952 0 0>, rot=<0.0013 0 0 -1>}
 thigh_r {parent=pelvis, loc=<-1.0853 -0.2109 8.7053>, rot=<0.0355 -0.0009 0.9993 0.0155>}
 calf_r {parent=thigh_r, loc=<42.5286 0.0019 0.7297>, rot=<0.0006 -0.0161 -0.0391 -0.9991>}
 calf_twist_01_r {parent=calf_r, loc=<35.8485 1.9197 0.5793>, rot=<0.0237 0.0095 -0.0035 -0.9997>}
 calf_twist_02_r {parent=calf_r, loc=<19.1546 1.7943 0.2638>, rot=<0.0237 0.0095 -0.0035 -0.9997>}
 foot_r {parent=calf_r, loc=<46.7403 2.5014 0.5919>, rot=<0.0427 -0.0238 0.0353 -0.9982>}
 ball_r {parent=foot_r, loc=<3.5658 -10.8066 0.1544>, rot=<0.0001 -0.0001 -0.7063 -0.7079>}
 dyn_shoelace_MSTR_r {parent=foot_r, loc=<-2.0857 -6.2559 -0.5611>, rot=<-0.018 -0.0417 0.9405 -0.3368>}
 dyn_shoelace_loopR_r {parent=dyn_shoelace_MSTR_r, loc=<0.3413 0.1535 -0.4883>, rot=<0.3832 -0.3114 -0.6571 -0.5696>}
 dyn_shoelace_stringR_r {parent=dyn_shoelace_loopR_r, loc=<0.6988 -0.1645 0.8907>, rot=<-0.0834 0.4974 0.0929 -0.8585>}
 dyn_shoelace_loopL_r {parent=dyn_shoelace_MSTR_r, loc=<0.2674 -0.0228 0.5421>, rot=<0.4263 -0.46 0.6146 0.4784>}
 dyn_shoelace_stringL_r {parent=dyn_shoelace_loopL_r, loc=<0.7545 -0.1893 -0.875>, rot=<0.065 -0.5365 0.0771 -0.8378>}
 thigh_twist_01_r {parent=thigh_r, loc=<27.952 0 0>, rot=<0.0013 0 0 -1>}
 dyn_glute_l {parent=pelvis, loc=<-7.3795 -8.2882 -9.0865>, rot=<0.491 -0.5089 0.5089 -0.491>}
 dyn_glute_r {parent=pelvis, loc=<-7.3795 -8.2882 9.087>, rot=<0.491 -0.5089 0.5089 -0.491>}
 dyn_hoodPocket_l {parent=pelvis, loc=<6.7798 6.0049 -10.2558>, rot=<-0.4008 -0.5826 -0.3796 -0.5966>}
]
