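# Physics asset for the left and right dangling strings on the female medium Tie Dye Fashion player outfit.
# Each string is a three-capsule chain that thickens toward the tip, with limited swing and tight twist between segments.
# Each chain's root is rotationally locked to spine_05, which links to spine_04 with free rotation.
# Two large capsules on those spine bones approximate the upper torso behind the strings.
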PhysicsAsset /BRCosmetics/Characters/Player/Female/Medium/Bodies/F_MED_Tie_Dye_Fashion/Meshes/F_MED_Tie_Dye_Fashion_String_Physics
bodies {
 dyn_string_l_01 [{shapeType=Capsule, length=2, radius=0.6, shapeName=dyn_string_l_01_capsule, center=<1.5 -0.1695 0>, rotation=<rot 90 0 0>}]
 dyn_string_l_02 [{shapeType=Capsule, length=2, radius=0.7, shapeName=dyn_string_l_02_capsule, center=<1.5 -0.1695 0>, rotation=<rot 90 0 0>}]
 dyn_string_l_03 [{shapeType=Capsule, length=2, radius=0.8, shapeName=dyn_string_l_03_capsule, center=<1.5 -0.2022 0>, rotation=<rot 90 0 0>}]
 dyn_string_r_01 [{shapeType=Capsule, length=2, radius=0.6, shapeName=dyn_string_r_01_capsule, center=<1.5 -0.1695 0>, rotation=<rot 90 0 0>}]
 dyn_string_r_02 [{shapeType=Capsule, length=2, radius=0.7, shapeName=dyn_string_r_02_capsule, center=<1.5 -0.1695 0>, rotation=<rot 90 0 0>}]
 dyn_string_r_03 [{shapeType=Capsule, length=2, radius=0.8, shapeName=dyn_string_r_03_capsule, center=<1.5 -0.1695 0>, rotation=<rot 90 0 0>}]
 spine_04 [{shapeType=Capsule, length=15.7309, radius=9.7, shapeName=spine_04_capsule, center=<0.0128 2.9525 0.027>}]
 spine_05 [{shapeType=Capsule, length=12.5702, radius=9.25, shapeName=spine_05_capsule, center=<-3.2281 1.3846 -0.2696>}]
}
constraints [
 {bone1Name=dyn_string_r_02, bone2Name=dyn_string_r_01, swing1LimitDegrees=35, swing2LimitDegrees=35, twistLimitDegrees=5}
 {bone1Name=dyn_string_r_03, bone2Name=dyn_string_r_02, swing1LimitDegrees=35, swing2LimitDegrees=35, twistLimitDegrees=5}
 {bone1Name=dyn_string_l_02, bone2Name=dyn_string_l_01, swing1LimitDegrees=35, swing2LimitDegrees=35, twistLimitDegrees=5}
 {bone1Name=dyn_string_l_03, bone2Name=dyn_string_l_02, swing1LimitDegrees=35, swing2LimitDegrees=35, twistLimitDegrees=5}
 {
  bone1Name=dyn_string_r_01
  bone2Name=spine_05
  swing1LimitDegrees=45
  swing1Motion=Locked
  swing2LimitDegrees=45
  swing2Motion=Locked
  twistLimitDegrees=45
  twistMotion=Locked
 }
 {
  bone1Name=dyn_string_l_01
  bone2Name=spine_05
  swing1LimitDegrees=45
  swing1Motion=Locked
  swing2LimitDegrees=45
  swing2Motion=Locked
  twistLimitDegrees=45
  twistMotion=Locked
 }
 {
  bone1Name=spine_05
  bone2Name=spine_04
  swing1LimitDegrees=45
  swing1Motion=Free
  swing2LimitDegrees=45
  swing2Motion=Free
  twistLimitDegrees=45
  twistMotion=Free
 }
]
