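# Post-process-style AnimBP for the female medium Tie-Dye Fashion body, adding secondary motion to strings, shoelaces and hoodie cloth.
# It copies the attached parent's pose, runs a RigidBody string simulation, then chains per-bone AnimDynamics for shoelaces, hood and hoodie, most spring-driven.
# An additive left hood-pocket offset and outfit Control Rig finish the pose.
# EventGraph updates gravity, and a once-only backbling check sets weight variables.
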AnimBlueprint /BRCosmetics/Characters/Player/Female/Medium/Bodies/F_MED_Tie_Dye_Fashion/Meshes/F_MED_Tie_Dye_Fashion_Skeleton_AnimBP
target_skeleton=&/F_MED_Tie_Dye_Fashion_Skeleton
graphs [
 AnimGraph {nodes [
   n1 = RigidBody!(
    ComponentPose=Local To Component(LocalPose=Copy Pose From Mesh() {bUseAttachedParent=true})
    Alpha
    OverrideWorldGravity:Vector="0,0,0"
    ExternalForce
   ) -> (Pose) {
    bOverrideWorldGravity=true
    overridePhysicsAsset=&/F_MED_Tie_Dye_Fashion_String_Physics
    simulationSpace=BaseBoneSpace
    alphaBoolBlend {blendOption=Linear}
    baseBoneRef {boneName=spine_02}
    componentLinearAccScale=<0.3 0.3 0.5>
    componentLinearVelScale=<0.2 0.2 0.3>
    overrideWorldGravity {z=-980}
   }
   n2 = Anim Dynamics - Bone: dyn_shoelace_loopL_l(ComponentPose=@n1, Alpha=1, ExternalForce="0,0,0") {
    angularDampingOverride=0.99
    bOverrideAngularDamping=true
    bUseGravityOverride=true
    alphaBoolBlend {blendOption=Linear}
    boundBone {boneName=dyn_shoelace_loopL_l}
    componentAppliedLinearAccClamp=<3000 3000 3000>
    componentLinearAccScale=<0.15 0.15 0.15>
    componentLinearVelScale=<0.15 0.15 0.15>
    gravityOverride {z=-980}
    physicsBodyDefinitions [{
      boundBone {boneName=dyn_shoelace_loopL_l}
      boxExtents=<17 17 17>
      constraintSetup {angularLimitsMax=<5 20 20>, angularLimitsMin=<-10 -30 -30>, angularTarget=<0 0 0>}
      localJointOffset=<2 2 2>
     }]
   }
   n3 = Anim Dynamics - Bone: dyn_shoelace_stringL_l(ComponentPose=@n2, Alpha=1, ExternalForce="0,0,0") {
    angularDampingOverride=0.995
    bOverrideAngularDamping=true
    bUseGravityOverride=true
    alphaBoolBlend {blendOption=Linear}
    boundBone {boneName=dyn_shoelace_stringL_l}
    componentAppliedLinearAccClamp=<3000 3000 3000>
    componentLinearAccScale=<0.15 0.15 0.15>
    componentLinearVelScale=<0.15 0.15 0.15>
    gravityOverride {z=-980}
    physicsBodyDefinitions [{
      boundBone {boneName=dyn_shoelace_stringL_l}
      boxExtents=<15 10 10>
      constraintSetup {angularLimitsMax=<5 20 0>, angularLimitsMin=<-10 -30 -30>, angularTarget=<0 0 0>}
      localJointOffset=<2 1 0.5>
     }]
   }
   n4 = Anim Dynamics - Bone: dyn_shoelace_loopR_l(ComponentPose=@n3, Alpha=1, ExternalForce="0,0,0") {
    angularDampingOverride=0.99
    bOverrideAngularDamping=true
    bUseGravityOverride=true
    alphaBoolBlend {blendOption=Linear}
    boundBone {boneName=dyn_shoelace_loopR_l}
    componentAppliedLinearAccClamp=<3000 3000 3000>
    componentLinearAccScale=<0.15 0.15 0.15>
    componentLinearVelScale=<0.15 0.15 0.15>
    gravityOverride {z=-980}
    physicsBodyDefinitions [{
      boundBone {boneName=dyn_shoelace_loopR_l}
      boxExtents=<17 17 17>
      constraintSetup {angularLimitsMax=<20 30 20>, angularLimitsMin=<-10 -20 -30>, angularTarget=<0 0 0>}
      localJointOffset=<3 2 -2>
     }]
   }
   n5 = Anim Dynamics - Bone: dyn_shoelace_stringR_l(ComponentPose=@n4, Alpha=1, ExternalForce="0,0,0") {
    angularDampingOverride=0.995
    bOverrideAngularDamping=true
    bUseGravityOverride=true
    alphaBoolBlend {blendOption=Linear}
    boundBone {boneName=dyn_shoelace_stringR_l}
    componentAppliedLinearAccClamp=<3000 3000 3000>
    componentLinearAccScale=<0.15 0.15 0.15>
    componentLinearVelScale=<0.15 0.15 0.15>
    gravityOverride {z=-980}
    physicsBodyDefinitions [{
      boundBone {boneName=dyn_shoelace_stringR_l}
      boxExtents=<15 10 10>
      constraintSetup {angularLimitsMax=<5 30 5>, angularLimitsMin=<-10 -20 -30>, angularTarget=<0 0 0>}
      localJointOffset=<2 1 0>
     }]
   }
   n6 = Anim Dynamics - Bone: dyn_shoelace_loopR_r(ComponentPose=@n5, Alpha=1, ExternalForce="0,0,0") {
    angularDampingOverride=0.99
    bOverrideAngularDamping=true
    bUseGravityOverride=true
    alphaBoolBlend {blendOption=Linear}
    boundBone {boneName=dyn_shoelace_loopR_r}
    componentAppliedLinearAccClamp=<3000 3000 3000>
    componentLinearAccScale=<0.15 0.15 0.15>
    componentLinearVelScale=<0.15 0.15 0.15>
    gravityOverride {z=-980}
    physicsBodyDefinitions [{
      boundBone {boneName=dyn_shoelace_loopR_r}
      boxExtents=<17 17 17>
      constraintSetup {angularLimitsMax=<5 20 20>, angularLimitsMin=<-10 -30 -30>, angularTarget=<0 0 0>}
      localJointOffset=<2 2 2>
     }]
   }
   n7 = Anim Dynamics - Bone: dyn_shoelace_stringR_r(ComponentPose=@n6, Alpha=1, ExternalForce="0,0,0") {
    angularDampingOverride=0.995
    bOverrideAngularDamping=true
    bUseGravityOverride=true
    alphaBoolBlend {blendOption=Linear}
    boundBone {boneName=dyn_shoelace_stringR_r}
    componentAppliedLinearAccClamp=<3000 3000 3000>
    componentLinearAccScale=<0.15 0.15 0.15>
    componentLinearVelScale=<0.15 0.15 0.15>
    gravityOverride {z=-980}
    physicsBodyDefinitions [{
      boundBone {boneName=dyn_shoelace_stringR_r}
      boxExtents=<15 10 10>
      constraintSetup {angularLimitsMax=<5 20 10>, angularLimitsMin=<-10 -30 -30>, angularTarget=<0 0 0>}
      localJointOffset=<2 1 1>
     }]
   }
   n8 = Anim Dynamics - Bone: dyn_shoelace_loopL_r(ComponentPose=@n7, Alpha=1, ExternalForce="0,0,0") {
    angularDampingOverride=0.99
    bOverrideAngularDamping=true
    bUseGravityOverride=true
    alphaBoolBlend {blendOption=Linear}
    boundBone {boneName=dyn_shoelace_loopL_r}
    componentAppliedLinearAccClamp=<3000 3000 3000>
    componentLinearAccScale=<0.15 0.15 0.15>
    componentLinearVelScale=<0.15 0.15 0.15>
    gravityOverride {z=-980}
    physicsBodyDefinitions [{
      boundBone {boneName=dyn_shoelace_loopL_r}
      boxExtents=<17 17 17>
      constraintSetup {angularLimitsMax=<20 30 20>, angularLimitsMin=<-10 -20 -30>, angularTarget=<0 0 0>}
      localJointOffset=<3 2 -2>
     }]
   }
   n9 = Anim Dynamics - Bone: dyn_shoelace_stringL_r(ComponentPose=@n8, Alpha=1, ExternalForce="0,0,0") {
    angularDampingOverride=0.995
    bOverrideAngularDamping=true
    bUseGravityOverride=true
    alphaBoolBlend {blendOption=Linear}
    boundBone {boneName=dyn_shoelace_stringL_r}
    componentAppliedLinearAccClamp=<3000 3000 3000>
    componentLinearAccScale=<0.15 0.15 0.15>
    componentLinearVelScale=<0.15 0.15 0.15>
    gravityOverride {z=-980}
    physicsBodyDefinitions [{
      boundBone {boneName=dyn_shoelace_stringL_r}
      boxExtents=<15 10 10>
      constraintSetup {angularLimitsMax=<5 10 5>, angularLimitsMin=<-10 -20 -30>, angularTarget=<0 0 0>}
      localJointOffset=<2 2 -2>
     }]
   }
   n10 = Anim Dynamics - Bone: dyn_hood_r_02!(ComponentPose=@n9, Alpha=1, GravityOverride:Vector="0,0,0", ExternalForce="0,0,0") -> (Pose) {
    angularDampingOverride=1
    angularSpringConstant=2000
    bAngularSpring=true
    bGravityOverrideInSimSpace=true
    bOverrideAngularDamping=true
    bUseGravityOverride=true
    simulationSpace=RootRelative
    alphaBoolBlend {blendOption=Linear}
    boundBone {boneName=dyn_hood_r_02}
    componentAppliedLinearAccClamp=<3000 3000 3000>
    componentLinearAccScale=<0.1 0.3 0.5>
    componentLinearVelScale=<0.1 0.3 0.5>
    gravityOverride {z=-980}
    physicsBodyDefinitions [{
      sphereCollisionRadius=0
      boundBone {boneName=dyn_hood_r_02}
      boxExtents=<10 20 20>
      constraintSetup {angularLimitsMax=<3 5 10>, angularLimitsMin=<-3 -55 -10>, angularTarget=<1 0 0>}
      localJointOffset=<5 0 5>
     }]
   }
   n11 = Anim Dynamics - Bone: dyn_hood_l_02!(ComponentPose=@n10, Alpha=1, GravityOverride:Vector="0,0,0", ExternalForce="0,0,0") -> (Pose) {
    angularDampingOverride=1
    angularSpringConstant=2000
    bAngularSpring=true
    bGravityOverrideInSimSpace=true
    bOverrideAngularDamping=true
    bUseGravityOverride=true
    simulationSpace=RootRelative
    alphaBoolBlend {blendOption=Linear}
    boundBone {boneName=dyn_hood_l_02}
    componentAppliedLinearAccClamp=<3000 3000 3000>
    componentLinearAccScale=<0.1 0.3 0.5>
    componentLinearVelScale=<0.1 0.3 0.5>
    gravityOverride {z=-980}
    physicsBodyDefinitions [{
      sphereCollisionRadius=0
      boundBone {boneName=dyn_hood_l_02}
      boxExtents=<10 20 20>
      constraintSetup {angularLimitsMax=<3 5 10>, angularLimitsMin=<-3 -55 -10>, angularTarget=<1 0 0>}
      localJointOffset=<5 0 5>
     }]
   }
   n12 = Anim Dynamics - Bone: dyn_hood_r_01!(ComponentPose=@n11, Alpha=1, GravityOverride:Vector="0,0,0", ExternalForce="0,0,0") -> (Pose) {
    angularDampingOverride=1
    angularSpringConstant=2000
    bAngularSpring=true
    bGravityOverrideInSimSpace=true
    bOverrideAngularDamping=true
    bUseGravityOverride=true
    simulationSpace=RootRelative
    alphaBoolBlend {blendOption=Linear}
    boundBone {boneName=dyn_hood_r_01}
    componentAppliedLinearAccClamp=<3000 3000 3000>
    componentLinearAccScale=<0.1 0.3 0.5>
    componentLinearVelScale=<0.1 0.3 0.5>
    gravityOverride {z=-980}
    physicsBodyDefinitions [{
      sphereCollisionRadius=0
      boundBone {boneName=dyn_hood_r_01}
      boxExtents=<10 20 20>
      constraintSetup {angularLimitsMax=<1 5 5>, angularLimitsMin=<-1 -40 -5>, angularTarget=<1 0 0>}
      localJointOffset=<5 0 0>
     }]
   }
   n13 = Anim Dynamics - Bone: dyn_hood_l_01!(ComponentPose=@n12, Alpha=1, GravityOverride:Vector="0,0,0", ExternalForce="0,0,0") -> (Pose) {
    angularDampingOverride=1
    angularSpringConstant=2000
    bAngularSpring=true
    bGravityOverrideInSimSpace=true
    bOverrideAngularDamping=true
    bUseGravityOverride=true
    simulationSpace=RootRelative
    alphaBoolBlend {blendOption=Linear}
    boundBone {boneName=dyn_hood_l_01}
    componentAppliedLinearAccClamp=<3000 3000 3000>
    componentLinearAccScale=<0.1 0.3 0.5>
    componentLinearVelScale=<0.1 0.3 0.5>
    gravityOverride {z=-980}
    physicsBodyDefinitions [{
      sphereCollisionRadius=0
      boundBone {boneName=dyn_hood_l_01}
      boxExtents=<10 20 20>
      constraintSetup {angularLimitsMax=<1 5 5>, angularLimitsMin=<-1 -40 -5>, angularTarget=<1 0 0>}
      localJointOffset=<5 0 0>
     }]
   }
   n14 = Anim Dynamics - Bone: dyn_hoodie_l!(ComponentPose=@n13, Alpha=0.5, GravityOverride:Vector="0,0,0", ExternalForce="0,0,0") -> (Pose) {
    angularDampingOverride=1
    bAngularSpring=true
    bGravityOverrideInSimSpace=true
    bOverrideAngularDamping=true
    bUseGravityOverride=true
    simulationSpace=RootRelative
    alphaBoolBlend {blendOption=Linear}
    boundBone {boneName=dyn_hoodie_l}
    componentAppliedLinearAccClamp=<3000 3000 3000>
    componentLinearAccScale=<0.3 0.3 0.3>
    componentLinearVelScale=<0.3 0.3 0.3>
    gravityOverride {z=-980}
    physicsBodyDefinitions [{
      sphereCollisionRadius=0
      boundBone {boneName=dyn_hoodie_l}
      boxExtents=<24 24 12>
      constraintSetup {angularLimitsMax=<5 0 5>, angularLimitsMin=<-5 -60 -5>, angularTarget=<1 0 0>}
      localJointOffset=<0 0 -12>
     }]
   }
   n15 = Anim Dynamics - Bone: dyn_hoodie_fr!(ComponentPose=@n14, Alpha=0.5, GravityOverride:Vector="0,0,0", ExternalForce="0,0,0") -> (Pose) {
    angularDampingOverride=1
    bAngularSpring=true
    bGravityOverrideInSimSpace=true
    bOverrideAngularDamping=true
    bUseGravityOverride=true
    simulationSpace=RootRelative
    alphaBoolBlend {blendOption=Linear}
    boundBone {boneName=dyn_hoodie_fr}
    componentAppliedLinearAccClamp=<3000 3000 3000>
    componentLinearAccScale=<0.3 0.3 0.3>
    componentLinearVelScale=<0.3 0.3 0.3>
    gravityOverride {z=-980}
    physicsBodyDefinitions [{
      sphereCollisionRadius=0
      boundBone {boneName=dyn_hoodie_fr}
      boxExtents=<24 24 12>
      constraintSetup {angularLimitsMax=<5 0 5>, angularLimitsMin=<-5 -60 -5>, angularTarget=<1 0 0>}
      localJointOffset=<0 0 -12>
     }]
   }
   n16 = Anim Dynamics - Bone: dyn_hoodie_r!(ComponentPose=@n15, Alpha=0.5, GravityOverride:Vector="0,0,0", ExternalForce="0,0,0") -> (Pose) {
    angularDampingOverride=1
    bAngularSpring=true
    bGravityOverrideInSimSpace=true
    bOverrideAngularDamping=true
    bUseGravityOverride=true
    simulationSpace=RootRelative
    alphaBoolBlend {blendOption=Linear}
    boundBone {boneName=dyn_hoodie_r}
    componentAppliedLinearAccClamp=<3000 3000 3000>
    componentLinearAccScale=<0.3 0.3 0.3>
    componentLinearVelScale=<0.3 0.3 0.3>
    gravityOverride {z=-980}
    physicsBodyDefinitions [{
      sphereCollisionRadius=0
      boundBone {boneName=dyn_hoodie_r}
      boxExtents=<24 24 12>
      constraintSetup {angularLimitsMax=<5 0 5>, angularLimitsMin=<-5 -60 -5>, angularTarget=<1 0 0>}
      localJointOffset=<0 0 -12>
     }]
   }
   n17 = Anim Dynamics - Bone: dyn_hoodie_bk!(ComponentPose=@n16, Alpha=0.5, GravityOverride:Vector="0,0,0", ExternalForce="0,0,0") -> (Pose) {
    angularDampingOverride=1
    bAngularSpring=true
    bGravityOverrideInSimSpace=true
    bOverrideAngularDamping=true
    bUseGravityOverride=true
    simulationSpace=RootRelative
    alphaBoolBlend {blendOption=Linear}
    boundBone {boneName=dyn_hoodie_bk}
    componentAppliedLinearAccClamp=<3000 3000 3000>
    componentLinearAccScale=<0.3 0.3 0.3>
    componentLinearVelScale=<0.3 0.3 0.3>
    gravityOverride {z=-980}
    physicsBodyDefinitions [{
      sphereCollisionRadius=0
      boundBone {boneName=dyn_hoodie_bk}
      boxExtents=<24 24 12>
      constraintSetup {angularLimitsMax=<5 0 5>, angularLimitsMin=<-5 -60 -5>, angularTarget=<1 0 0>}
      localJointOffset=<0 0 -12>
     }]
   }
   n18 = Anim Dynamics - Bone: dyn_upperhoodie_l_bt!(ComponentPose=@n17, Alpha=1, GravityOverride:Vector="0,0,0", ExternalForce="0,0,0") -> (Pose) {
    angularDampingOverride=1
    angularSpringConstant=20000
    bAngularSpring=true
    bGravityOverrideInSimSpace=true
    bOverrideAngularDamping=true
    bUseGravityOverride=true
    simulationSpace=RootRelative
    alphaBoolBlend {blendOption=Linear}
    boundBone {boneName=dyn_upperhoodie_l_bt}
    componentAppliedLinearAccClamp=<3000 3000 3000>
    componentLinearAccScale=<0.1 0.1 0.1>
    componentLinearVelScale=<0.1 0.1 0.1>
    gravityOverride {z=-980}
    physicsBodyDefinitions [{
      sphereCollisionRadius=0
      boundBone {boneName=dyn_upperhoodie_l_bt}
      boxExtents=<25 25 25>
      constraintSetup {angularLimitsMax=<5 10 5>, angularLimitsMin=<-5 -10 -5>, angularTarget=<1 0 0>}
      localJointOffset=<20 0 0>
     }]
   }
   n19 = Anim Dynamics - Bone: dyn_upperhoodie_l_tp!(ComponentPose=@n18, Alpha=1, GravityOverride:Vector="0,0,0", ExternalForce="0,0,0") -> (Pose) {
    angularDampingOverride=1
    angularSpringConstant=20000
    bAngularSpring=true
    bGravityOverrideInSimSpace=true
    bOverrideAngularDamping=true
    bUseGravityOverride=true
    simulationSpace=RootRelative
    alphaBoolBlend {blendOption=Linear}
    boundBone {boneName=dyn_upperhoodie_l_tp}
    componentAppliedLinearAccClamp=<3000 3000 3000>
    componentLinearAccScale=<0.1 0.1 0.1>
    componentLinearVelScale=<0.1 0.1 0.1>
    gravityOverride {z=-980}
    physicsBodyDefinitions [{
      sphereCollisionRadius=0
      boundBone {boneName=dyn_upperhoodie_l_tp}
      boxExtents=<25 25 25>
      constraintSetup {angularLimitsMax=<5 10 5>, angularLimitsMin=<-5 -10 -5>, angularTarget=<1 0 0>}
      localJointOffset=<20 0 0>
     }]
   }
   n20 = Anim Dynamics - Bone: dyn_upperhoodie_l_bk!(ComponentPose=@n19, Alpha=1, GravityOverride:Vector="0,0,0", ExternalForce="0,0,0") -> (Pose) {
    angularDampingOverride=1
    angularSpringConstant=20000
    bAngularSpring=true
    bGravityOverrideInSimSpace=true
    bOverrideAngularDamping=true
    bUseGravityOverride=true
    simulationSpace=RootRelative
    alphaBoolBlend {blendOption=Linear}
    boundBone {boneName=dyn_upperhoodie_l_bk}
    componentAppliedLinearAccClamp=<3000 3000 3000>
    componentLinearAccScale=<0.1 0.1 0.1>
    componentLinearVelScale=<0.1 0.1 0.1>
    gravityOverride {z=-980}
    physicsBodyDefinitions [{
      sphereCollisionRadius=0
      boundBone {boneName=dyn_upperhoodie_l_bk}
      boxExtents=<25 25 25>
      constraintSetup {angularLimitsMax=<5 10 5>, angularLimitsMin=<-5 -10 -5>, angularTarget=<1 0 0>}
      localJointOffset=<20 0 0>
     }]
   }
   n21 = Anim Dynamics - Bone: dyn_upperhoodie_l_ft!(ComponentPose=@n20, Alpha=1, GravityOverride:Vector="0,0,0", ExternalForce="0,0,0") -> (Pose) {
    angularDampingOverride=1
    angularSpringConstant=20000
    bAngularSpring=true
    bGravityOverrideInSimSpace=true
    bOverrideAngularDamping=true
    bUseGravityOverride=true
    simulationSpace=RootRelative
    alphaBoolBlend {blendOption=Linear}
    boundBone {boneName=dyn_upperhoodie_l_ft}
    componentAppliedLinearAccClamp=<3000 3000 3000>
    componentLinearAccScale=<0.1 0.1 0.1>
    componentLinearVelScale=<0.1 0.1 0.1>
    gravityOverride {z=-980}
    physicsBodyDefinitions [{
      sphereCollisionRadius=0
      boundBone {boneName=dyn_upperhoodie_l_ft}
      boxExtents=<25 25 25>
      constraintSetup {angularLimitsMax=<5 10 5>, angularLimitsMin=<-5 -10 -5>, angularTarget=<1 0 0>}
      localJointOffset=<20 0 0>
     }]
   }
   n22 = Anim Dynamics - Bone: dyn_upperhoodie_r_bt!(ComponentPose=@n21, Alpha=1, GravityOverride:Vector="0,0,0", ExternalForce="0,0,0") -> (Pose) {
    angularDampingOverride=1
    angularSpringConstant=20000
    bAngularSpring=true
    bGravityOverrideInSimSpace=true
    bOverrideAngularDamping=true
    bUseGravityOverride=true
    simulationSpace=RootRelative
    alphaBoolBlend {blendOption=Linear}
    boundBone {boneName=dyn_upperhoodie_r_bt}
    componentAppliedLinearAccClamp=<3000 3000 3000>
    componentLinearAccScale=<0.1 0.1 0.1>
    componentLinearVelScale=<0.1 0.1 0.1>
    gravityOverride {z=-980}
    physicsBodyDefinitions [{
      sphereCollisionRadius=0
      boundBone {boneName=dyn_upperhoodie_r_bt}
      boxExtents=<25 25 25>
      constraintSetup {angularLimitsMax=<5 10 5>, angularLimitsMin=<-5 -10 -5>, angularTarget=<1 0 0>}
      localJointOffset=<20 0 0>
     }]
   }
   n23 = Anim Dynamics - Bone: dyn_upperhoodie_r_tp!(ComponentPose=@n22, Alpha=1, GravityOverride:Vector="0,0,0", ExternalForce="0,0,0") -> (Pose) {
    angularDampingOverride=1
    angularSpringConstant=20000
    bAngularSpring=true
    bGravityOverrideInSimSpace=true
    bOverrideAngularDamping=true
    bUseGravityOverride=true
    simulationSpace=RootRelative
    alphaBoolBlend {blendOption=Linear}
    boundBone {boneName=dyn_upperhoodie_r_tp}
    componentAppliedLinearAccClamp=<3000 3000 3000>
    componentLinearAccScale=<0.1 0.1 0.1>
    componentLinearVelScale=<0.1 0.1 0.1>
    gravityOverride {z=-980}
    physicsBodyDefinitions [{
      sphereCollisionRadius=0
      boundBone {boneName=dyn_upperhoodie_r_tp}
      boxExtents=<25 25 25>
      constraintSetup {angularLimitsMax=<5 10 5>, angularLimitsMin=<-5 -10 -5>, angularTarget=<1 0 0>}
      localJointOffset=<20 0 0>
     }]
   }
   n24 = Anim Dynamics - Bone: dyn_upperhoodie_r_bk!(ComponentPose=@n23, Alpha=1, GravityOverride:Vector="0,0,0", ExternalForce="0,0,0") -> (Pose) {
    angularDampingOverride=1
    angularSpringConstant=20000
    bAngularSpring=true
    bGravityOverrideInSimSpace=true
    bOverrideAngularDamping=true
    bUseGravityOverride=true
    simulationSpace=RootRelative
    alphaBoolBlend {blendOption=Linear}
    boundBone {boneName=dyn_upperhoodie_r_bk}
    componentAppliedLinearAccClamp=<3000 3000 3000>
    componentLinearAccScale=<0.1 0.1 0.1>
    componentLinearVelScale=<0.1 0.1 0.1>
    gravityOverride {z=-980}
    physicsBodyDefinitions [{
      sphereCollisionRadius=0
      boundBone {boneName=dyn_upperhoodie_r_bk}
      boxExtents=<25 25 25>
      constraintSetup {angularLimitsMax=<5 10 5>, angularLimitsMin=<-5 -10 -5>, angularTarget=<1 0 0>}
      localJointOffset=<20 0 0>
     }]
   }
   n25 = Anim Dynamics - Bone: dyn_upperhoodie_r_ft!(ComponentPose=@n24, Alpha=1, GravityOverride:Vector="0,0,0", ExternalForce="0,0,0") -> (Pose) {
    angularDampingOverride=1
    angularSpringConstant=20000
    bAngularSpring=true
    bGravityOverrideInSimSpace=true
    bOverrideAngularDamping=true
    bUseGravityOverride=true
    simulationSpace=RootRelative
    alphaBoolBlend {blendOption=Linear}
    boundBone {boneName=dyn_upperhoodie_r_ft}
    componentAppliedLinearAccClamp=<3000 3000 3000>
    componentLinearAccScale=<0.1 0.1 0.1>
    componentLinearVelScale=<0.1 0.1 0.1>
    gravityOverride {z=-980}
    physicsBodyDefinitions [{
      sphereCollisionRadius=0
      boundBone {boneName=dyn_upperhoodie_r_ft}
      boxExtents=<25 25 25>
      constraintSetup {angularLimitsMax=<5 10 5>, angularLimitsMin=<-5 -10 -5>, angularTarget=<1 0 0>}
      localJointOffset=<20 0 0>
     }]
   }
   n26 = Anim Dynamics - Bone: dyn_lowerhoodie_l_bk!(ComponentPose=@n25, Alpha=1, GravityOverride:Vector="0,0,0", ExternalForce="0,0,0") -> (Pose) {
    angularDampingOverride=1
    angularSpringConstant=20000
    bAngularSpring=true
    bGravityOverrideInSimSpace=true
    bOverrideAngularDamping=true
    bUseGravityOverride=true
    simulationSpace=RootRelative
    alphaBoolBlend {blendOption=Linear}
    boundBone {boneName=dyn_lowerhoodie_l_bk}
    componentAppliedLinearAccClamp=<3000 3000 3000>
    componentLinearAccScale=<0.1 0.1 0.1>
    componentLinearVelScale=<0.1 0.1 0.1>
    gravityOverride {z=-980}
    physicsBodyDefinitions [{
      sphereCollisionRadius=0
      boundBone {boneName=dyn_lowerhoodie_l_bk}
      boxExtents=<20 20 20>
      constraintSetup {angularLimitsMax=<5 15 5>, angularLimitsMin=<-5 -15 -5>, angularTarget=<1 0 0>}
      localJointOffset=<10 0 0>
     }]
   }
   n27 = Anim Dynamics - Bone: dyn_lowerhoodie_l_bt!(ComponentPose=@n26, Alpha=1, GravityOverride:Vector="0,0,0", ExternalForce="0,0,0") -> (Pose) {
    angularDampingOverride=1
    angularSpringConstant=20000
    bAngularSpring=true
    bGravityOverrideInSimSpace=true
    bOverrideAngularDamping=true
    bUseGravityOverride=true
    simulationSpace=RootRelative
    alphaBoolBlend {blendOption=Linear}
    boundBone {boneName=dyn_lowerhoodie_l_bt}
    componentAppliedLinearAccClamp=<3000 3000 3000>
    componentLinearAccScale=<0.1 0.1 0.1>
    componentLinearVelScale=<0.1 0.1 0.1>
    gravityOverride {z=-980}
    physicsBodyDefinitions [{
      sphereCollisionRadius=0
      boundBone {boneName=dyn_lowerhoodie_l_bt}
      boxExtents=<20 20 20>
      constraintSetup {angularLimitsMax=<5 15 5>, angularLimitsMin=<-5 -15 -5>, angularTarget=<1 0 0>}
      localJointOffset=<10 0 0>
     }]
   }
   n28 = Anim Dynamics - Bone: dyn_lowerhoodie_l_fr!(ComponentPose=@n27, Alpha=1, GravityOverride:Vector="0,0,0", ExternalForce="0,0,0") -> (Pose) {
    angularDampingOverride=1
    angularSpringConstant=20000
    bAngularSpring=true
    bGravityOverrideInSimSpace=true
    bOverrideAngularDamping=true
    bUseGravityOverride=true
    simulationSpace=RootRelative
    alphaBoolBlend {blendOption=Linear}
    boundBone {boneName=dyn_lowerhoodie_l_fr}
    componentAppliedLinearAccClamp=<3000 3000 3000>
    componentLinearAccScale=<0.1 0.1 0.1>
    componentLinearVelScale=<0.1 0.1 0.1>
    gravityOverride {z=-980}
    physicsBodyDefinitions [{
      sphereCollisionRadius=0
      boundBone {boneName=dyn_lowerhoodie_l_fr}
      boxExtents=<20 20 20>
      constraintSetup {angularLimitsMax=<5 15 5>, angularLimitsMin=<-5 -15 -5>, angularTarget=<1 0 0>}
      localJointOffset=<10 0 0>
     }]
   }
   n29 = Anim Dynamics - Bone: dyn_lowerhoodie_l_tp!(ComponentPose=@n28, Alpha=1, GravityOverride:Vector="0,0,0", ExternalForce="0,0,0") -> (Pose) {
    angularDampingOverride=1
    angularSpringConstant=20000
    bAngularSpring=true
    bGravityOverrideInSimSpace=true
    bOverrideAngularDamping=true
    bUseGravityOverride=true
    simulationSpace=RootRelative
    alphaBoolBlend {blendOption=Linear}
    boundBone {boneName=dyn_lowerhoodie_l_tp}
    componentAppliedLinearAccClamp=<3000 3000 3000>
    componentLinearAccScale=<0.1 0.1 0.1>
    componentLinearVelScale=<0.1 0.1 0.1>
    gravityOverride {z=-980}
    physicsBodyDefinitions [{
      sphereCollisionRadius=0
      boundBone {boneName=dyn_lowerhoodie_l_tp}
      boxExtents=<20 20 20>
      constraintSetup {angularLimitsMax=<5 15 5>, angularLimitsMin=<-5 -15 -5>, angularTarget=<1 0 0>}
      localJointOffset=<10 0 0>
     }]
   }
   n30 = Anim Dynamics - Bone: dyn_lowerhoodie_r_bk!(ComponentPose=@n29, Alpha=1, GravityOverride:Vector="0,0,0", ExternalForce="0,0,0") -> (Pose) {
    angularDampingOverride=1
    angularSpringConstant=20000
    bAngularSpring=true
    bGravityOverrideInSimSpace=true
    bOverrideAngularDamping=true
    bUseGravityOverride=true
    simulationSpace=RootRelative
    alphaBoolBlend {blendOption=Linear}
    boundBone {boneName=dyn_lowerhoodie_r_bk}
    componentAppliedLinearAccClamp=<3000 3000 3000>
    componentLinearAccScale=<0.1 0.1 0.1>
    componentLinearVelScale=<0.1 0.1 0.1>
    gravityOverride {z=-980}
    physicsBodyDefinitions [{
      sphereCollisionRadius=0
      boundBone {boneName=dyn_lowerhoodie_r_bk}
      boxExtents=<20 20 20>
      constraintSetup {angularLimitsMax=<5 15 5>, angularLimitsMin=<-5 -15 -5>, angularTarget=<1 0 0>}
      localJointOffset=<10 0 0>
     }]
   }
   n31 = Anim Dynamics - Bone: dyn_lowerhoodie_r_bt!(ComponentPose=@n30, Alpha=1, GravityOverride:Vector="0,0,0", ExternalForce="0,0,0") -> (Pose) {
    angularDampingOverride=1
    angularSpringConstant=20000
    bAngularSpring=true
    bGravityOverrideInSimSpace=true
    bOverrideAngularDamping=true
    bUseGravityOverride=true
    simulationSpace=RootRelative
    alphaBoolBlend {blendOption=Linear}
    boundBone {boneName=dyn_lowerhoodie_r_bt}
    componentAppliedLinearAccClamp=<3000 3000 3000>
    componentLinearAccScale=<0.1 0.1 0.1>
    componentLinearVelScale=<0.1 0.1 0.1>
    gravityOverride {z=-980}
    physicsBodyDefinitions [{
      sphereCollisionRadius=0
      boundBone {boneName=dyn_lowerhoodie_r_bt}
      boxExtents=<20 20 20>
      constraintSetup {angularLimitsMax=<5 15 5>, angularLimitsMin=<-5 -15 -5>, angularTarget=<1 0 0>}
      localJointOffset=<10 0 0>
     }]
   }
   n32 = Anim Dynamics - Bone: dyn_lowerhoodie_r_fr!(ComponentPose=@n31, Alpha=1, GravityOverride:Vector="0,0,0", ExternalForce="0,0,0") -> (Pose) {
    angularDampingOverride=1
    angularSpringConstant=20000
    bAngularSpring=true
    bGravityOverrideInSimSpace=true
    bOverrideAngularDamping=true
    bUseGravityOverride=true
    simulationSpace=RootRelative
    alphaBoolBlend {blendOption=Linear}
    boundBone {boneName=dyn_lowerhoodie_r_fr}
    componentAppliedLinearAccClamp=<3000 3000 3000>
    componentLinearAccScale=<0.1 0.1 0.1>
    componentLinearVelScale=<0.1 0.1 0.1>
    gravityOverride {z=-980}
    physicsBodyDefinitions [{
      sphereCollisionRadius=0
      boundBone {boneName=dyn_lowerhoodie_r_fr}
      boxExtents=<20 20 20>
      constraintSetup {angularLimitsMax=<5 15 5>, angularLimitsMin=<-5 -15 -5>, angularTarget=<1 0 0>}
      localJointOffset=<10 0 0>
     }]
   }
   n33 = Anim Dynamics - Bone: dyn_lowerhoodie_r_tp!(ComponentPose=@n32, Alpha=1, GravityOverride:Vector="0,0,0", ExternalForce="0,0,0") -> (Pose) {
    angularDampingOverride=1
    angularSpringConstant=20000
    bAngularSpring=true
    bGravityOverrideInSimSpace=true
    bOverrideAngularDamping=true
    bUseGravityOverride=true
    simulationSpace=RootRelative
    alphaBoolBlend {blendOption=Linear}
    boundBone {boneName=dyn_lowerhoodie_r_tp}
    componentAppliedLinearAccClamp=<3000 3000 3000>
    componentLinearAccScale=<0.1 0.1 0.1>
    componentLinearVelScale=<0.1 0.1 0.1>
    gravityOverride {z=-980}
    physicsBodyDefinitions [{
      sphereCollisionRadius=0
      boundBone {boneName=dyn_lowerhoodie_r_tp}
      boxExtents=<20 20 20>
      constraintSetup {angularLimitsMax=<5 15 5>, angularLimitsMin=<-5 -15 -5>, angularTarget=<1 0 0>}
      localJointOffset=<10 0 0>
     }]
   }
   n34 = "Transform (Modify) Bone - Bone: dyn_hoodPocket_l"(ComponentPose=@n33, Alpha=1, Translation="-1.2,0,1.2", Rotation="0,0,0", Scale="1,1,1") {
    translationMode=BMM_Additive
    translationSpace=BCS_BoneSpace
    alphaBoolBlend {blendOption=Linear}
    alphaScaleBiasClamp {bInterpResult=true}
    boneToModify {boneName=dyn_hoodPocket_l}
    translation {x=-1.2, z=1.2}
   }
   n35 = Control Rig(Source=Component To Local(ComponentPose=@n34)) {
    alphaBoolBlend {blendOption=Linear}
    controlRigAssetReference {blueprintRigClass=&/F_MED_Tie_Dye_Fashion_CtrlRig.F_MED_Tie_Dye_Fashion_CtrlRig_C}
   }
   Output Pose(Result=@n35) {layerGroup=None}
  ]}
 EventGraph {nodes [
   Event BlueprintUpdateAnimation()
   > Set Gravity Override(Gravity Override="0, 0,-980")
   > Do Once()
   n1 = CheckForBackbling()
   > n2 = Branch(Condition=NOR Boolean!(A:bool=@n1, B:bool=@n1.bIsBackblingLong) -> (ReturnValue:bool))
   > n3 = Set isWearingBackbling(isWearingBackbling:double=0.4) -> (Output_Get:double=0)
   Set sleeveAlpha(execute=[@n3.then, @n4.then], sleeveAlpha:double=0.3) -> (Output_Get:double)
   n4 = Set isWearingBackbling(execute=@n2.else, isWearingBackbling=0) -> (Output_Get=0)
   TryGetPawnOwner()
  ]}
]
variables [
 Gravity Override {type=Vector}
 isWearingBackbling {type=number}
 sleeveAlpha {type=number}
]
